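# Unlit translucent trail material for the Cadet skydiving contrail, usable on Niagara ribbons and sprites.
# Panning cloud textures distort the graffiti texture's UVs, weighted toward low U.
# Cloud noise, a scrolling graffiti mask and a UV1 gradient blend emissive from PaintColor to particle color tint and erode depth-faded opacity.
# An altitude-driven clock flipbook branch, time-driven in frontend preview, appears unused.
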Material /BRCosmetics/Athena/Effects/Particles/Skydiving/NS_Contrail_Cadet/M_Contrail_Cadet_Trail
graph {
 properties {
  bUsedWithNiagaraRibbons=true
  bUsedWithNiagaraSprites=true
  blendMode=BLEND_Translucent
  pixelDepthOffsetMode=PDOM_Legacy
  shadingModel=MSM_Unlit
  usedShadingModels=Unlit
 }
 nodes [
  n1 = TextureCoordinate()
  AppendVector(A=ComponentMask(None=@n1) {r=true}, B=ComponentMask(None=@n1) {g=true})
  n2 = TextureSample(UVs=Panner(Coordinate=TextureCoordinate() {uTiling=0.2, vTiling=1.5}, Time=Time()) {speedX=-1}) {samplerType=SAMPLERTYPE_Normal, texture=/Game/Global/Textures/Default/Tileable_Patterns/Cloud_normals}
  ComponentMask(None=@n2.RGB) {g=true}
  n3 = ComponentMask(None=@n1) {r=true}
  n4 = TextureSample(UVs=Panner(Coordinate=TextureCoordinate() {uTiling=0.5}, Time=Time()) {speedX=-1.5}) {samplerType=SAMPLERTYPE_LinearColor, texture=/Game/VisualThreatManager/StormVisuals/Textures/Clouds_VectorDisplacement}
  n5 = Multiply(A=Multiply(A=@n2.B, B=@n4.G) {constA=1}) {constB=0.05}
  n6 = Multiply(A=Power(Base=OneMinus(None=@n3)), B=@n5) {constB=0.05}
  n7 = TextureSample(UVs=Saturate(None=Add(A=@n1, B=@n6))) {texture=&/T_Contrail_Cadet_Graffiti_01}
  n8 = VectorParameter() {parameterName=PaintColor, defaultValue=<rgba 0.3545 0 0.5312 1>}
  n9 = ParticleColor()
  n10 = Panner(Coordinate=TextureCoordinate() {uTiling=0.3, vTiling=2}, Time=Time()) {speedX=-1}
  n11 = TextureCoordinate() {coordinateIndex=1}
  n12 = Panner(Coordinate=@n11, Time=Time()) {speedY=3}
  n13 = Multiply(
   A=Multiply(
    A=TextureSample(UVs=@n10) {samplerType=SAMPLERTYPE_LinearColor, texture=&Clouds_VectorDisplacement}.B
    B=@n4.G
   ) {constA=1}
   B=OneMinus(None=TextureSampleParameter2D(UVs=@n12) {parameterName=TrailTexture_BChannel, texture=/BRCosmetics/Weapons/FORT_Melee/Pickaxe_Cadet_Female/FX/T_Pickaxe_Pickaxe_Graffiti}.G)
  ) {constA=1}
  n14 = ComponentMask(None=@n11) {r=true}
  n15 = Power(Base=Add(A=@n13, B=Multiply(A=@n14) {constB=5})) {constExponent=5}
  MaterialFunctionCall(Emissive=LinearInterpolate(A=Multiply(A=@n7.RGB, B=@n8.RGB), B=Multiply(A=@n7.RGB, B=@n9.RGB) {constB=0.15}, Alpha=Saturate(None=@n15))) {materialFunction=&MF_FX_EmissiveAdjustments}
  n16 = CollectionParameter() {collection=&FortniteMaterialParameters, parameterName=AthenaFrontendPreview}
  n17 = ObjectPositionWS()
  n18 = LinearInterpolate(
   A=MaterialFunctionCall(
    Input=ComponentMask(None=@n17) {b=true}
    Input Low=Constant() {r=12000}
    Input High=ScalarParameter() {defaultValue=75000, parameterName=FullRange}
    Target Low=Constant()
    Target High=Constant() {r=1}
   ) {materialFunction=&RemapValueRange}.Result
   B=Divide(A=Time()) {constB=24}
   Alpha=@n16
  )
  n19 = ScalarParameter() {defaultValue=-0.0833, parameterName=RangeOffset}
  n20 = Multiply(A=Frac(None=Add(A=@n18, B=@n19))) {constB=12}
  n21 = AppendVector(A=Floor(None=@n20), B=Floor(None=Divide(A=@n20) {constB=4}))
  n22 = TextureSample(UVs=Add(A=Multiply(A=@n21) {constB=0.25}, B=TextureCoordinate() {uTiling=0.25, vTiling=0.25})) {texture=/BRCosmetics/Effects/Fort_Effects/Textures/T_ClockTextSubUV}
  Multiply(B=Add(A=@n22.R, B=Multiply(A=@n22.G) {constB=25}))
  Multiply(B=Multiply(A=@n22.R) {constB=0.15})
  OneMinus(None=@n5)
  n23 = Add(A=Multiply(A=Power(Base=@n13)) {constA=1, constB=3}, B=@n14)
  n24 = DepthFade(
   Opacity=Multiply(A=Multiply(A=@n9.A, B=@n7.A) {constB=0.15}, B=Saturate(None=@n23)) {constB=0.15}
   FadeDistance=ScalarParameter() {defaultValue=20, group=DepthDade, parameterName=DepthFadeDistance, sortPriority=1}
  )
  ShadingPathSwitch(Default=@n24, Deferred=@n24) {desc=Migrated from FeatureLevelSwitch}
 ]
}
