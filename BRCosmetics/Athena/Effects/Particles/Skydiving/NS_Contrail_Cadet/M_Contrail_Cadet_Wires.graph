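# Two-sided lit material for wire meshes in the Cadet skydiving contrail's Niagara mesh particles, adding a swaying rotation around each particle's pivot.
# A local axis spins over time with per-particle random phase, then blends with its base direction to form the sway axis.
# A sine drives the small rotation angle, phase-shifted by a dynamic time offset and lagged along the UVs.
# Quantized second-UV noise also scales that offset, and parameters define the metallic surface.
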Material /BRCosmetics/Athena/Effects/Particles/Skydiving/NS_Contrail_Cadet/M_Contrail_Cadet_Wires
graph {
 properties {
  bTangentSpaceNormal=false
  bUsedWithNiagaraMeshParticles=true
  pixelDepthOffsetMode=PDOM_Legacy
  twoSided=true
  usedShadingModels=Default Lit
 }
 nodes [
  n1 = ComponentMask(None=TextureCoordinate() {coordinateIndex=1}) {g=true}
  n2 = Multiply(A=Floor(None=Multiply(A=@n1) {constB=100})) {constB=0.5}
  n3 = DynamicParameter() {defaultValue=<rgba 1 5 15 22>, paramNames [Time Offset, Chain Width, Sphere Mask Offset, Sphere Mask Radius]}
  Multiply(A=ComponentMask(None=VectorNoise(World Position=@n2)) {r=true}, B=@n3.Time Offset)
  n4 = Constant3Vector() {desc=rotation axis' rotation axis, constant=<rgba 1 0 0 0>}
  n5 = Add(A=Time(), B=ParticleRandom())
  n6 = ScalarParameter() {defaultValue=0.618, parameterName=Axis Rotation Speed}
  n7 = Normalize(VectorInput=Transform(None=Constant3Vector() {desc=rotation axis, constant=<rgba 0 1 0 0>}) {transformSourceType=TRANSFORMSOURCE_Local})
  n8 = RotateAboutAxis(
   NormalizedRotationAxis=Normalize(VectorInput=Transform(None=@n4) {transformSourceType=TRANSFORMSOURCE_Local})
   RotationAngle=Add(A=@n3.Time Offset, B=Multiply(A=@n5, B=@n6))
   PivotPoint=Constant()
   Position=@n7
  )
  n9 = Time()
  n10 = ScalarParameter() {defaultValue=2.618, parameterName=Rotation Speed}
  n11 = ComponentMask(None=TextureCoordinate()) {r=true}
  n12 = OneMinus(None=ComponentMask(None=TextureCoordinate() {coordinateIndex=1}) {r=true})
  n13 = Subtract(
   A=Add(A=Multiply(A=@n9, B=@n10), B=@n3.Time Offset)
   B=Multiply(A=Max(A=@n11, B=@n12), B=ScalarParameter() {defaultValue=0.86, parameterName=Height Time Offset})
  )
  RotateAboutAxis(
   NormalizedRotationAxis=Normalize(VectorInput=Add(A=@n8, B=@n7))
   RotationAngle=Multiply(A=Sine(None=@n13), B=ScalarParameter() {defaultValue=0.03, parameterName=Rotation Angle Max})
   PivotPoint=MaterialFunctionCall() {materialFunction=&ObjectPivotPoint}.Mesh Particle Pivot Location
   Position=WorldPosition().XYZ
  )
  ScalarParameter() {defaultValue=0.7, parameterName=Metallic}
  ScalarParameter() {defaultValue=0.2, parameterName=Roughness}
  ScalarParameter() {parameterName=Specular}
  ScalarParameter() {defaultValue=0.45, parameterName=Time Offset Amount}
  VectorParameter() {parameterName=Base Color, defaultValue=<rgba 0.6354 0.6354 0.6354 0.428>}
 ]
}
