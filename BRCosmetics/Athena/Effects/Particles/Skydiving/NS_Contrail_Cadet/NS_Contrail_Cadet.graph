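# Niagara skydiving contrail for the Cadet cosmetic, trailing paint and mesh effects from the player's skeleton, mainly the hands.
# Two paint ribbon layers trail two bones, the first colored per side with speed-scaled rise and time-of-day alpha.
# GPU multicolor drops, soft purple hand sprites and spinning bullet meshes add scatter, while persistent burst-spawned chain meshes follow the wrists.
# Front-end modifiers swap scale, velocity and material values between lobby and gameplay.
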NiagaraSystem /BRCosmetics/Athena/Effects/Particles/Skydiving/NS_Contrail_Cadet/NS_Contrail_Cadet
dependencies {
 usedDataInterfaces [
  &NiagaraDataInterfaceColorCurve
  &NiagaraDataInterfaceCurve
  &NiagaraDataInterfaceArrayInt32
  &NiagaraDataInterfaceArrayFloat3
 ]
 usedDynamicInputs [&Lerp_Float, &SetBoolByFloatComparison]
 usedModules [
  &SystemState
  /Game/Effects/Niagara/Modules/Skydiving/FN_DivingAttributes/v5/FN_DivingAttributes
  &self:SystemSpawnScript.SystemScriptSource.SystemScriptGraph.NiagaraNodeAssignment_2.SetVariables_#1
  &self:SystemSpawnScript.SystemScriptSource.SystemScriptGraph.NiagaraNodeAssignment_3.SetVariables_#2
  &FN_Integrator
  &self:SystemSpawnScript.SystemScriptSource.SystemScriptGraph.NiagaraNodeAssignment_4.SetVariables_#3
 ]
}
emitters {
 Chains {
  emitterSpawnScript [&FN_PreserveSkydiveUserVariables() {module=PreserveUserVariables}]
  emitterUpdateScript [
   &EmitterState(Inactive Response=::NewEnumerator1, Life Cycle Mode=::NewEnumerator1, Loop Duration=0, UseLoopDelay=false)
   &SpawnBurst_Instantaneous(Spawn Count=6, Spawn Probability=HlslExpression() {hlslExpression="Emitter.LoopCount == 0"}, Use Spawn Probability=true) {module=SpawnBurst_Instantaneous001}
   &SpawnBurst_Instantaneous(Spawn Count=6, Spawn Time=3.5626, Use Spawn Probability=false, UseLoopCountLimit=false)
  ]
  particleSpawnScript [
   &FN_SetSpawnGroup(Number of Spawn Groups=2)
   &InitializeParticle(
    Color Mode=::NewEnumerator0
    Lifetime=0
    Sprite Size Mode=::NewEnumerator3
    Uniform Sprite Size=40
    UsePositionOffset=false
    Write Material Random=true
   ) {module=InitializeParticle001}
   &Color(Color=$NPC.FortniteNPC.FortniteActiveTimeOfDay)
   &DynamicMaterialParameters(
    Index 0 Param 1=&RandomRangeFloat(Minimum=0, Maximum=10)
    Index 0 Param 2=&FN_FrontEndFloatModifier(Gameplay=5, Front End=4)
    Index 0 Param 3=&FN_FrontEndFloatModifier(Gameplay=0, Front End=15)
    Index 0 Param 4=&FN_FrontEndFloatModifier(Gameplay=55, Front End=22)
    Module.Param0WriteEnabled=true
    Module.Param1WriteEnabled=true
    Module.Param2WriteEnabled=true
    Module.Param3WriteEnabled=true
   )
   /SetVariables(
    Particles.MeshIndex=&SelectIntFromArray(
     Direct Array Index=&ReturnParticleIDAttributes()
     Direct Array Mode=::NewEnumerator1
     Int Selection Array=DataInterface() {propertyValues {IntData [0,1,1,2,2,0]}}
    )
    Particles.Scale=&FN_FrontEndVector3Modifier(Gameplay=<1 1 1>, Front End=<0.8 0.8 0.8>)
   ) {isSetParameters=true}
   &InitialMeshOrientation(
    Mesh Orientation Mode=::NewEnumerator7
    Rotation=&MakeVector(X=&RandomRangeFloat(Minimum=0, Maximum=1), Y=0, Z=0)
    Rotation Coordinate Space=::NewEnumerator0
    Use Rotation Vector=true
   ) {enabled=false}
   &InitialMeshOrientation(Mesh Axis To Orient=<0 1 0>, Mesh Orientation Mode=::NewEnumerator2, Rotation=<0 -0.25 0>, Use Rotation Vector=true) {module=InitialMeshOrientation001}
  ]
  particleUpdateScript [
   &ParticleState(Kill Particles When Lifetime Has Elapsed=false)
   &SkeletalMeshLocation(
    Bone / Socket Index=$Particles.SpawnGroup
    Bone Sampling Mode=::NewEnumerator4
    Inherit Velocity=false
    Interpolate Mesh Sampling=false
    MaskSpawnGroup=false
    Sampled Position Offset=&SelectVectorFromArray(
     Direct Array Index=&ReturnParticleIDAttributes()
     Direct Array Mode=::NewEnumerator1
     Vector Selection Array=DataInterface() {propertyValues {FloatData []}}
    )
    Sampled Position Offset Coordinate Space:ENiagaraMeshTransforms=::NewEnumerator2
    Skeletal Mesh=$User.SkeletalMesh Wrist
   )
   /Niagara/Modules/Update/Orientation/V2/OrientMeshToVector(
    Look At Direction:Vector3f=&NormalizeVector(
     Vector To Normalize=&Add_Vector(
      A=&NormalizeVector(
       Vector To Normalize=&FN_FrontEndVector3Modifier(
        Gameplay=&Multiply_VectorByFloat(Vector=$Engine.Owner.Velocity, Float=-1)
        Front End=&TransformVector(
         Vector=&NormalizeVector(
          Vector To Normalize=&Lerp_Vector(A=<0 -0.2 1>, B=<0 -2 1>, Alpha=$User.ParaGlideLeanAlpha)
          Safe Normalize=true
          Safe Fallback Threshold=0
          Safe Fallback Vector=<0 0 0.001>
         )
         Destination Space=::Simulation
        )
       )
       Safe Normalize=false
       Safe Fallback Threshold=0
       Safe Fallback Vector=<0 0 0.001>
      )
      B=&Multiply_VectorByFloat(
       Vector=&SelectVectorFromArray(
        Direct Array Index=&ReturnParticleIDAttributes()
        Direct Array Mode=::NewEnumerator1
        Vector Selection Array=DataInterface() {propertyValues {FloatData [<1 0 0>, <1 0 0>, <0 1 0>, <0 1 0>, <-1 0 0>, <-1 0 0>, <0 -1 0>, <0 -1 0>]}}
       )
       Float=&RandomRangeFloat(Minimum=0.166, Maximum=0.266)
      )
     )
     Safe Normalize=false
     Safe Fallback Threshold=0
     Safe Fallback Vector=<0 0 0.001>
    )
    Override Quat=false
    Side Axis Direction:Vector3f=&TransformVector(Vector=<0 1 0>, Destination Space=::World)
    UseOptionalSideAxis=false
    UseOptionalUpAxis=false
   ) {enabled=false}
   &UpdateMeshOrientation(
    Rotation Coordinate Space=::NewEnumerator3
    Rotation Rate=&RandomRangeFloat(Minimum=-1, Maximum=1)
    Rotation Vector=<1 0 0>
   ) {enabled=false}
  ]
  rendererClasses [&NiagaraMeshRendererProperties]
  renderers [{rendererClass=&NiagaraMeshRendererProperties, rendererIndex=0}]
 }
 Drops {
  simTarget=GPUComputeSim
  emitterUpdateScript [
   &EmitterState(Inactive Response=::NewEnumerator1, Life Cycle Mode=::NewEnumerator1, Loop Behavior=::NewEnumerator1, Loop Duration=3600, UseLoopDelay=false)
   &SpawnRate(
    SpawnGroup:NiagaraInt32=&RandomRangeInt(Minimum=0, Maximum=1)
    SpawnRate=50
    Use Spawn Probability=false
   )
  ]
  particleSpawnScript [
   &FN_SetSpawnGroup(Number of Spawn Groups=2)
   &InitializeParticle(
    Color=&MakeCustomLinearColorFromBool(
     Bool=&RandomBool()
     True Color=&MakeCustomLinearColorFromBool(Bool=&RandomBool(), True Color=<rgba 0.6715 0 0.8 1>, False Color=<rgba 1 0.5211 0 1>)
     False Color=&MakeCustomLinearColorFromBool(Bool=&RandomBool(), True Color=<rgba 0 0.7043 1 1>, False Color=<rgba 1 0 0.3437 1>)
    )
    Lifetime Max=0.5
    Lifetime Min=0.5
    Lifetime Mode=::NewEnumerator1
    Sprite Size Mode=::NewEnumerator1
    Uniform Sprite Size Min=10
    UsePositionOffset=false
    Write Material Random=true
   ) {module=InitializeParticle001}
   &SkeletalMeshLocation(
    Bone / Socket Index=$Particles.SpawnGroup
    Bone Sampling Mode=::NewEnumerator4
    Inherit Velocity=false
    Interpolate Mesh Sampling=false
    MaskSpawnGroup=false
    Sampled Position Offset Coordinate Space:ENiagaraMeshTransforms=::NewEnumerator3
    Skeletal Mesh=$User.SkeletalMesh
   )
   /SetVariables(Particles.Velocity=&UniformRangedVector(Minimum=<-100 -100 250>, Maximum=<100 100 400>)) {isSetParameters=true}
   /SetVariables(Particles.SpriteSize=&Multiply_Vector2(
     A=<5 10>
     B=&Vector2DFromFloat(Value=&UniformRangedFloat(Minimum=1, Maximum=2))
    )) {isSetParameters=true}
  ]
  particleUpdateScript [
   &ParticleState()
   &GravityForce(Gravity=<0 0 -100>)
   &Drag(Use Linear Drag=true, Use Rotational Drag=false)
   &SolveForcesAndVelocity(Clamp Velocity=false, Limit Acceleration=false)
   &ScaleColor(
    Scale Alpha=&FloatFromCurve(FloatCurve=DataInterface() {propertyValues {Curve {keys [
         <at 0 0 RCIM_Cubic>
         {arriveTangent=1.6041, leaveTangent=1.6041, time=0.05, value=1.0113}
         {interpMode=RCIM_Cubic, leaveTangent=-1.0633, time=0.75, value=1}
         <at 0.99 0 RCIM_Cubic>
        ]}}})
    Scale RGB=<0.6 0.8 1>
    ScaleA=true
    ScaleRGB=true
   )
   &ScaleSpriteSize(Scale Factor=&Vector2DFromFloat(Value=&FloatFromCurve(
      FloatCurve=DataInterface() {propertyValues {Curve {keys [
          {arriveTangent=-0.4919, interpMode=RCIM_Cubic, leaveTangent=-0.4919, tangentMode=RCTM_User, time=0.25, value=2}
          {arriveTangent=-3.3301, interpMode=RCIM_Cubic, leaveTangent=-3.3302, tangentMode=RCTM_User, time=1, value=0}
         ]}}}
      Scale Curve=0.65
     )))
  ]
  rendererClasses [&NiagaraSpriteRendererProperties]
  renderers [{rendererClass=&NiagaraSpriteRendererProperties}]
 }
 Hand_Bullets {
  emitterSpawnScript [&FN_PreserveSkydiveUserVariables() {module=PreserveUserVariables}]
  emitterUpdateScript [
   &EmitterState()
   &SpawnRate(SpawnRate=10, Use Spawn Probability=false)
  ]
  particleSpawnScript [
   /Game/Effects/Niagara/Modules/Groups/FN_SetSpawnGroup_v1/FN_SetSpawnGroup(Emitter Bone Count=4) {enabled=false}
   /Game/Effects/Niagara/Modules/Groups/FN_SetSpawnGroup(Number of Spawn Groups=2) {module=FN_SetSpawnGroup001}
   /Niagara/Modules/Spawn/Initialization/InitializeParticle(
    Lifetime=&UniformRangedFloat(Minimum=0.6, Maximum=1.2)
    Mesh Scale=&RandomRangeVector(Minimum=<0.9 0.9 0.9>, Maximum=<0.6 0.6 0.6>)
    Module.Write Color=true
    Module.Write Lifetime=true
    Module.Write Mass=true
    Module.Write Material Random=true
    Module.Write Position=false
    Module.Write Scale=true
    Module.Write SpriteRotation=false
    Module.Write SpriteSize=false
    Position=&SimulationPosition()
   )
   &SampleSkeletalMeshSkeleton(Mesh:NiagaraDataInterfaceSkeletalMesh=$User.SkeletalMesh, Specific Bone Index:NiagaraInt32=$Particles.SpawnGroup) {enabled=false}
   /Niagara/Modules/Spawn/Location/SkeletalMeshSkeletonLocation(Module.Inherit Velocity=false, Module.Use Sampled Bone as Ribbon ID=true) {enabled=false}
   &SkeletalMeshLocation(
    Bone / Socket Index=$Particles.SpawnGroup
    Bone Sampling Mode=::NewEnumerator4
    Inherit Velocity=false
    MaskSpawnGroup=false
    Sampled Position Offset=<0 0 -80>
    Skeletal Mesh=$User.SkeletalMesh
   )
   &InitialMeshOrientation(Use Rotation Vector=false)
   /SetVariables(Particles.Position=&Add_Vector(
     A=Linked() {linkedVariable {name=Particles.Position, type=NiagaraPosition}}
     B=&Multiply_VectorByFloat(
      Vector=/Niagara/DynamicInputs/Helpers/SafeNormalizeVector(Vector To Normalize:Vector3f=$Particles.Velocity, Fallback Vector:Vector3f=<0 0 1>)
      Float=HlslExpression() {hlslExpression="Particles.SpawnGroup < 2 ?  95: 60"}
     )
    )) {isSetParameters=true}
   &AddVelocityInCone(
    Cone Angle=40
    Cone Axis=&SafeNormalizeVector(
     Vector To Normalize:Vector3f=&Multiply_VectorByFloat(Vector=$Particles.Velocity, Float=-1)
     Fallback Vector:Vector3f=<0 0 1>
    )
    Cone Axis Coordinate Space=::World
    Module.Use Velocity Falloff On Cone Axis=false
    Velocity Strength=&UniformRangedFloat(Minimum=150, Maximum=400)
   )
   &MeshRotationForce(
    "Lever Radius (cm)":NiagaraFloat=&UniformRangedFloat(Minimum=150, Maximum=350)
    Rotation=/Niagara/DynamicInputs/Random/RandomVector(
     Vector Scale:NiagaraFloat=1
     Constraint Vector:Vector3f=<0 0 1>
     Randomness Mode:ENiagaraRandomnessMode=::NewEnumerator0
     Random Seed:NiagaraInt32=0
     Fixed Random Seed:NiagaraBool=false
    )
   )
   &ApplyInitialForces()
   /SetVariables(
    Particles.MeshIndex=&SelectIntFromArray(
     Direct Array Index=&ReturnParticleIDAttributes()
     Direct Array Mode=::NewEnumerator1
     Int Selection Array=DataInterface() {propertyValues {IntData [0,1,1,2,2,0]}}
    )
    Particles.Scale=&FN_FrontEndVector3Modifier(Gameplay=<1 1 1>, Front End=<0.8 0.8 0.8>)
   ) {isSetParameters=true}
  ]
  particleUpdateScript [
   &ParticleState()
   &ScaleMeshSize(Scale Factor=&VectorFromFloat(Value=&FloatFromCurve(FloatCurve=DataInterface() {propertyValues {Curve {keys [
          <at 0.0008 0.667 RCIM_Cubic>
          {interpMode=RCIM_Cubic, tangentMode=RCTM_User, time=0.1657, value=1.0109}
          {interpMode=RCIM_Cubic, tangentMode=RCTM_User, time=0.7195, value=1.0073}
          {arriveTangent=-4.4535, interpMode=RCIM_Cubic, leaveTangent=-4.4535, tangentMode=RCTM_User, time=1.0011, value=0.002}
         ]}}})))
   &SolveForcesAndVelocity(Clamp Velocity=false, Limit Acceleration=false)
   &SolveRotationalForcesAndVelocity()
  ]
  rendererClasses [&NiagaraMeshRendererProperties]
  renderers [{rendererClass=&NiagaraMeshRendererProperties, rendererIndex=0}]
 }
 Paint_Ribbon01 {
  emitterUpdateScript [
   &EmitterState()
   /Game/Effects/Niagara/Modules/Ribbons/FN_SpawnRibbon_v3(
    Number of Ribbons=2
    Skeleton Mesh=DataInterface() {propertyValues {
      DefaultMesh=null
      ExcludeBoneName=root
      OptionalComponentClass=null
      PreviewMesh=&M_Med_Soldier_04_Preview
      SkinningMode=SkinOnTheFly
      SoftSourceActor=null
      SourceMode=Default
      UvSetIndex=0
      WholeMeshLOD=-1
      bExcludeBone=false
      bReadDeformedGeometry=true
      bRequireCurrentFrameData=true
      ComponentTags []
      FilteredBones [hand_l,hand_r]
      FilteredSockets []
      MeshUserParameter {parameter {name=None, typeDefHandle {registeredTypeIndex=1}, varData []}}
      SamplingRegions []
     }}
   )
  ]
  particleSpawnScript [
   &FN_SetSpawnGroup(Number of Spawn Groups=2)
   /Game/Effects/Niagara/Modules/Ribbons/FN_InitializeRibbon_v5(Enable Skeletal Position=true, Lifetime=1, Ribbon Width=50, SkeletalMesh:NiagaraDataInterfaceSkeletalMesh=$User.SkeletalMesh)
   &AddVelocity(Velocity=&FN_FrontEndVector3Modifier(Gameplay=<0 0 300>, Front End=<0 0 400>)) {module=AddVelocity001}
   /SetVariables(Particles.Color=&LerpLinearColors(
     StartColor=&ColorFromCurve(
      ColorCurve=DataInterface() {propertyValues {
        AlphaCurve {keys [<at 0 1 RCIM_Cubic>]}
        BlueCurve {keys [<at 0.1027 0.651>,<at 1 0.1955 RCIM_Cubic>]}
        GreenCurve {keys [<at 0.1027 0>,<at 1 0.0254 RCIM_Cubic>]}
        RedCurve {keys [<at 0.1027 0.3373>,<at 1 1 RCIM_Cubic>]}
       }}
      CurveIndex=&RandomRangeFloat(Minimum=0, Maximum=1)
     )
     EndColor=&ColorFromCurve(
      ColorCurve=DataInterface() {propertyValues {
        AlphaCurve {keys [<at 0 1 RCIM_Cubic>]}
        BlueCurve {keys [<at 0.0929 0.651>,<at 1 1 RCIM_Cubic>]}
        GreenCurve {keys [<at 0.0929 0>,<at 1 0.0833 RCIM_Cubic>]}
        RedCurve {keys [<at 0.0929 0.3373>,<at 1 0.5317 RCIM_Cubic>]}
       }}
      CurveIndex=&RandomRangeFloat(Minimum=0, Maximum=1)
     )
     LerpFactor=&MakeFloatFromInt(Int=$Particles.SpawnGroup)
    )) {isSetParameters=true, module=SetVariables_#4}
   /SetVariables(Particles.MaterialRandom:NiagaraFloat=&RandomRangeFloat(Minimum=0, Maximum=1)) {isSetParameters=true}
   /SetVariables() {isSetParameters=true}
  ]
  particleUpdateScript [
   &ScaleVelocity(Velocity Scale=&MakeVector(
     X=0
     Y=0
     Z=&Multiply_Float(
      A=0.9
      B=&FloatFromCurve(
       FloatCurve=DataInterface() {propertyValues {Curve {keys [<at 0 0.9>,<at 1 3>]}}}
       CurveIndex=&Divide_Float(A=&VectorLength(Vector=$Engine.Owner.Velocity), B=9000)
      )
     )
    )) {module=ScaleVelocity001}
   &SolveForcesAndVelocity(Clamp Velocity=false, Limit Acceleration=false)
   /Game/Effects/Niagara/Modules/Ribbons/FN_UpdateRibbon_v6(Enable Skeletal Position=true, SkeletalMesh:NiagaraDataInterfaceSkeletalMesh=$User.SkeletalMesh)
   &DynamicMaterialParameters(
    Index 0 Param 1=&MaskFloatBySpawnGroup(Masked Float=0, Passthrough Float=0.5)
    Index 0 Param 2=&FloatFromCurve(FloatCurve=DataInterface() {propertyValues {Curve {keys [
         {arriveTangent=-0.3062, interpMode=RCIM_Cubic, leaveTangent=-0.3062, tangentMode=RCTM_User, time=0, value=0.5}
         {arriveTangent=0.0167, interpMode=RCIM_Cubic, leaveTangent=0.0167, tangentMode=RCTM_Break, time=1, value=0}
        ]}}})
    Index 0 Param 3=1
    Index 0 Param 4=1
    Module.Param0WriteEnabled=true
    Module.Param1WriteEnabled=true
    Module.Param2WriteEnabled=true
    Module.Param3WriteEnabled=true
   )
   &ScaleRibbonWidth(Ribbon Width Scale=&FloatFromCurve(
     FloatCurve=DataInterface() {propertyValues {Curve {keys [
         {arriveTangent=-0.0063, interpMode=RCIM_Cubic, leaveTangent=-0.0063, tangentMode=RCTM_User, time=0, value=1}
         {arriveTangent=2.3571, interpMode=RCIM_Cubic, leaveTangent=2.3571, time=0.4, value=2.025}
         <at 0.7 2.325>
        ]}}}
     Scale Curve=0.6
    ))
   &ScaleColor(
    Scale Alpha=&Multiply_Float(
     A=&FNTimeOfDayScalar(Night Scale=1.5, Morning Scale=2, Evening Scale=2)
     B=&FloatFromCurve(FloatCurve=DataInterface() {propertyValues {Curve {keys [<at 0 0>,<at 0.1 1>,<at 1 1>]}}})
    )
    Scale RGB=&Multiply_Vector(
     A=&MakeVectorFromLinearColorRGB(Color=<rgba 1 1 1 1>)
     B=&VectorFromFloat(Value=1.5)
    )
    ScaleA=true
    ScaleRGB=true
   )
  ]
  rendererClasses [&NiagaraRibbonRendererProperties]
  renderers [{rendererClass=&NiagaraRibbonRendererProperties}]
 }
 Paint_Ribbon02 {
  emitterUpdateScript [
   &EmitterState()
   &SpawnRate(SpawnRate=0, Use Spawn Probability=false) {enabled=false}
   &FN_SpawnRibbon_v3(
    Number of Ribbons=2
    Skeleton Mesh=DataInterface() {propertyValues {
      DefaultMesh=null
      ExcludeBoneName=root
      OptionalComponentClass=null
      PreviewMesh=&M_Med_Soldier_04_Preview
      SkinningMode=SkinOnTheFly
      SoftSourceActor=null
      SourceMode=Default
      UvSetIndex=0
      WholeMeshLOD=-1
      bExcludeBone=false
      bReadDeformedGeometry=true
      bRequireCurrentFrameData=true
      ComponentTags []
      FilteredBones [hand_l,hand_r]
      FilteredSockets []
      MeshUserParameter {parameter {name=None, typeDefHandle {registeredTypeIndex=1}, varData []}}
      SamplingRegions []
     }}
   )
  ]
  particleSpawnScript [
   &FN_SetSpawnGroup(Number of Spawn Groups=2)
   &FN_InitializeRibbon_v5(Enable Skeletal Position=true, Lifetime=0.4, Ribbon Width=50, SkeletalMesh:NiagaraDataInterfaceSkeletalMesh=$User.SkeletalMesh)
   &AddVelocity(Velocity=&FN_FrontEndVector3Modifier(Gameplay=<0 0 0>, Front End=<0 0 800>)) {module=AddVelocity001}
   /SetVariables(Particles.Color=&LerpLinearColors(
     StartColor=&ColorFromCurve(
      ColorCurve=DataInterface() {propertyValues {
        AlphaCurve {keys [<at 0 1 RCIM_Cubic>]}
        BlueCurve {keys [<at 0 1 RCIM_Cubic>,<at 0.0929 1>,<at 0.5021 0>,<at 0.7704 0.2097>,<at 1 0.3701 RCIM_Cubic>]}
        GreenCurve {keys [<at 0 0.7559 RCIM_Cubic>,<at 0.0929 0.0537>,<at 0.5021 0.8217>,<at 0.7704 0>,<at 1 0.0135 RCIM_Cubic>]}
        RedCurve {keys [<at 0 0 RCIM_Cubic>,<at 0.0929 0>,<at 0.5021 1>,<at 0.7704 0.651>,<at 1 1 RCIM_Cubic>]}
       }}
      CurveIndex=&RandomRangeFloat(Minimum=0, Maximum=1)
     )
     EndColor=&ColorFromCurve(
      ColorCurve=DataInterface() {propertyValues {
        AlphaCurve {keys [<at 0 1 RCIM_Cubic>]}
        BlueCurve {keys [<at 0 1 RCIM_Cubic>,<at 0.0929 1>,<at 0.5021 0>,<at 0.7704 0.2097>,<at 1 0.3701 RCIM_Cubic>]}
        GreenCurve {keys [<at 0 0.7559 RCIM_Cubic>,<at 0.0929 0.0537>,<at 0.5021 0.6901>,<at 0.7704 0>,<at 1 0.0135 RCIM_Cubic>]}
        RedCurve {keys [<at 0 0 RCIM_Cubic>,<at 0.0929 0>,<at 0.5021 1>,<at 0.7704 0.651>,<at 1 1 RCIM_Cubic>]}
       }}
      CurveIndex=&RandomRangeFloat(Minimum=0, Maximum=1)
     )
     LerpFactor=&MakeFloatFromInt(Int=$Particles.SpawnGroup)
    )) {enabled=false, isSetParameters=true, module=SetVariables_#4}
   /SetVariables(Particles.MaterialRandom:NiagaraFloat=&RandomRangeFloat(Minimum=0, Maximum=1)) {enabled=false, isSetParameters=true}
  ]
  particleUpdateScript [
   &ScaleVelocity(Velocity Scale=&Lerp_Vector(A=<0 0 1>, B=<1 1 1>, Alpha=0)) {module=ScaleVelocity001}
   &InheritSourceMovement(Applied Position Scale=<0 0 -1>, ApplyForce=false, ApplyPosition=true, ApplyPositionFalloff=false)
   &SolveForcesAndVelocity(Clamp Velocity=false, Limit Acceleration=false)
   &FN_UpdateRibbon_v6(Enable Skeletal Position=true, SkeletalMesh:NiagaraDataInterfaceSkeletalMesh=$User.SkeletalMesh)
   &ScaleRibbonWidth(Ribbon Width Scale=&FloatFromCurve(
     FloatCurve=DataInterface() {propertyValues {Curve {keys [
         {arriveTangent=-0.0063, interpMode=RCIM_Cubic, leaveTangent=-0.0063, tangentMode=RCTM_User, time=0, value=1}
         {arriveTangent=2.3571, interpMode=RCIM_Cubic, leaveTangent=2.3571, time=0.4, value=2.025}
         <at 0.7 2.325>
        ]}}}
     Scale Curve=0.6
    ))
   &ScaleColor(
    Scale Alpha=&Multiply_Float(
     A=&FNTimeOfDayScalar(Night Scale=1.5, Morning Scale=2, Evening Scale=2)
     B=&FloatFromCurve(FloatCurve=DataInterface() {propertyValues {Curve {keys [<at 0 0>,<at 0.1 1>,<at 1 1>]}}})
    )
    Scale RGB=&Multiply_Vector(
     A=&MakeVectorFromLinearColorRGB(Color=<rgba 1 1 1 1>)
     B=&VectorFromFloat(Value=2)
    )
    ScaleA=true
    ScaleRGB=true
   ) {enabled=false, module=ScaleColor001}
  ]
  rendererClasses [&NiagaraRibbonRendererProperties]
  renderers [{rendererClass=&NiagaraRibbonRendererProperties}]
 }
 StaticHandSprites {
  emitterSpawnScript [&FN_PreserveSkydiveUserVariables() {module=PreserveSkydiveUserVariables}]
  emitterUpdateScript [
   &EmitterScalability(
    MaxDistance=2000
    Module.bApplySpawnCountScale=false
    Module.bApplySpawnCountScaleByDistanceFraction=false
    Module.bUseMaxDistance=true
    Module.bUseMinDistance=true
    SpawnCountScaleByDistanceFraction=DataInterface() {propertyValues {Curve {keys [<at -0.0098 1>,<at 0.9896 0>]}}}
   ) {enabled=false}
   &EmitterLifeCycle(NextLoopDuration=1) {enabled=false}
   &EmitterState(Enable Distance Culling=true, MaxDistance=2000, Module.bUseMaxDistance=true, Module.bUseMinDistance=false, Scalability Mode=::NewEnumerator1)
   /Game/Effects/Niagara/Modules/Spawn/FN_SkydiveSpawnRate(Enable Feet=false, Global Spawnrate=10) {module=SkydiveSpawnRate}
  ]
  particleSpawnScript [
   /Game/Effects/Niagara/Modules/Groups/FN_SetSpawnGroup_v1/FN_SetSpawnGroup(Override SpawnGroup=true, SpawnGroup:NiagaraInt32=&SpawnGroup())
   &InitializeParticle(
    Lifetime=&UniformRangedFloat(Minimum=0.55, Maximum=0.8)
    Module.Write Color=false
    Module.Write Lifetime=true
    Module.Write Mass=false
    Module.Write Material Random=true
    Module.Write Position=false
    Module.Write Scale=false
    Module.Write SpriteRotation=true
    Module.Write SpriteSize=true
    Position=&SimulationPosition()
    Sprite Rotation=&UniformRangedFloat(Minimum=0, Maximum=360)
    Sprite Size=&Vector2DFromFloat(Value=&UniformRangedFloat(Minimum=19, Maximum=32))
   )
   /SetVariables(Particles.SpriteSize=&Multiply_Vector2(
     A=$Particles.SpriteSize
     B=&Vector2DFromFloat(Value=&OneMinusFloat(Float=$Emitter.Scalability.DistanceFraction))
    )) {isSetParameters=true, module=SetVariables_#5}
   &SampleSkeletalMeshSkeleton(
    Mesh:NiagaraDataInterfaceSkeletalMesh=$User.SkeletalMesh
    Specific Bone Index:NiagaraInt32=$Particles.SpawnGroup
    Specific Socket Index:NiagaraInt32=$Particles.SpawnGroup
   )
   &AddVelocity(Velocity=<0 0 200>)
   &DynamicMaterialParameters(Module.Param0WriteEnabled=true, Module.Param1WriteEnabled=true, Module.Param2WriteEnabled=true, Module.Param3WriteEnabled=true) {enabled=false}
   &SkeletalMeshSkeletonLocation(Module.Inherit Velocity=false, Module.Use Sampled Bone as Ribbon ID=true)
  ]
  particleUpdateScript [
   &Color(Color=&ColorFromCurve(ColorCurve=DataInterface() {propertyValues {
       AlphaCurve {keys [<at 0 0 RCIM_Cubic>,<at 0.2951 0.6 RCIM_Cubic>,<at 0.9829 0>]}
       BlueCurve {keys [<at 0 1 RCIM_Cubic>]}
       GreenCurve {keys [<at 0 0.0555 RCIM_Cubic>]}
       RedCurve {keys [<at 0 0.6375 RCIM_Cubic>]}
      }}))
   &ParticleState()
   &SolveForcesAndVelocity(Clamp Velocity=false, Limit Acceleration=false)
   &ScaleSpriteSize(Scale Factor=&Vector2DFromFloat(Value=&FloatFromCurve(FloatCurve=DataInterface() {propertyValues {Curve {keys [<at 0 0.5>,<at 1 1.4>]}}}, Scale Curve=2)))
  ]
  rendererClasses [&NiagaraSpriteRendererProperties]
  renderers [{rendererClass=&NiagaraSpriteRendererProperties}]
 }
 StaticHandSprites001 {
  emitterSpawnScript [&FN_PreserveSkydiveUserVariables() {module=PreserveSkydiveUserVariables}]
  emitterUpdateScript [
   &EmitterScalability(
    MaxDistance=2000
    Module.bApplySpawnCountScale=false
    Module.bApplySpawnCountScaleByDistanceFraction=false
    Module.bUseMaxDistance=true
    Module.bUseMinDistance=true
    SpawnCountScaleByDistanceFraction=DataInterface() {propertyValues {Curve {keys [<at -0.0098 1>,<at 0.9896 0>]}}}
   ) {enabled=false}
   &EmitterLifeCycle(NextLoopDuration=1) {enabled=false}
   &EmitterState(Enable Distance Culling=true, MaxDistance=2000, Module.bUseMaxDistance=true, Module.bUseMinDistance=false, Scalability Mode=::NewEnumerator1)
   &FN_SkydiveSpawnRate(Enable Feet=false, Global Spawnrate=10) {module=SkydiveSpawnRate}
  ]
  particleSpawnScript [
   &FN_SetSpawnGroup(Override SpawnGroup=true, SpawnGroup:NiagaraInt32=&SpawnGroup())
   &InitializeParticle(
    Lifetime=&UniformRangedFloat(Minimum=0.55, Maximum=0.8)
    Module.Write Color=false
    Module.Write Lifetime=true
    Module.Write Mass=false
    Module.Write Material Random=true
    Module.Write Position=false
    Module.Write Scale=false
    Module.Write SpriteRotation=true
    Module.Write SpriteSize=true
    Position=&SimulationPosition()
    Sprite Rotation=&UniformRangedFloat(Minimum=0, Maximum=360)
    Sprite Size=&Vector2DFromFloat(Value=&UniformRangedFloat(Minimum=19, Maximum=32))
   )
   /SetVariables(Particles.SpriteSize=&Multiply_Vector2(
     A=$Particles.SpriteSize
     B=&Vector2DFromFloat(Value=&OneMinusFloat(Float=$Emitter.Scalability.DistanceFraction))
    )) {isSetParameters=true, module=SetVariables_#5}
   &SampleSkeletalMeshSkeleton(
    Mesh:NiagaraDataInterfaceSkeletalMesh=$User.SkeletalMesh
    Specific Bone Index:NiagaraInt32=$Particles.SpawnGroup
    Specific Socket Index:NiagaraInt32=$Particles.SpawnGroup
   )
   &AddVelocity(Velocity=<0 0 300>)
   &DynamicMaterialParameters(Module.Param0WriteEnabled=true, Module.Param1WriteEnabled=true, Module.Param2WriteEnabled=true, Module.Param3WriteEnabled=true) {enabled=false}
   &SkeletalMeshSkeletonLocation(Module.Inherit Velocity=false, Module.Use Sampled Bone as Ribbon ID=true)
  ]
  particleUpdateScript [
   &Color(Color=&ColorFromCurve(ColorCurve=DataInterface() {propertyValues {
       AlphaCurve {keys [<at 0 0 RCIM_Cubic>,<at 0.2951 0.4 RCIM_Cubic>,<at 0.9829 0>]}
       BlueCurve {keys [<at 0 1 RCIM_Cubic>]}
       GreenCurve {keys [<at 0 0.0555 RCIM_Cubic>]}
       RedCurve {keys [<at 0 0.6375 RCIM_Cubic>]}
      }}))
   &ParticleState()
   &SolveForcesAndVelocity(Clamp Velocity=false, Limit Acceleration=false)
   &ScaleSpriteSize(Scale Factor=&Vector2DFromFloat(Value=&FloatFromCurve(FloatCurve=DataInterface() {propertyValues {Curve {keys [<at 0 0.3>,<at 1 1.4>]}}}, Scale Curve=3.5)))
  ]
  rendererClasses [&NiagaraSpriteRendererProperties]
  renderers [{rendererClass=&NiagaraSpriteRendererProperties}]
 }
}
summary {
 systemName=NS_Contrail_Cadet
 emitters [
  {emitterName=Paint_Ribbon01, rendererClasses [&NiagaraRibbonRendererProperties]}
  {emitterName=Paint_Ribbon02, rendererClasses [&NiagaraRibbonRendererProperties]}
  {emitterName=Drops, simTarget=GPUComputeSim, rendererClasses [&NiagaraSpriteRendererProperties]}
  {emitterName=StaticHandSprites, rendererClasses [&NiagaraSpriteRendererProperties]}
  {emitterName=StaticHandSprites001, rendererClasses [&NiagaraSpriteRendererProperties]}
  {emitterName=Chains, rendererClasses [&NiagaraMeshRendererProperties]}
  {emitterName=Hand_Bullets, rendererClasses [&NiagaraMeshRendererProperties]}
 ]
 userVariables [
  User.bIsFrontEnd {type=NiagaraBool}
  User.bIsFrontEndPreview {type=NiagaraBool}
  User.FEVec {type=Vector3f, description="", defaultValue {struct=Vector3f, value=<0 0 0>}}
  User.NormalizedNotifyProgress {type=NiagaraFloat, defaultValue=0, description=""}
  User.ParaGlideLeanAlpha {type=NiagaraFloat, defaultValue=0, description=""}
  User.PlayerSpeed {type=NiagaraFloat, defaultValue=0, description=""}
  User.RibbonIDOffset {type=NiagaraInt32}
  User.SkeletalMesh {
   type=NiagaraDataInterfaceSkeletalMesh
   defaultValue {
    struct=NiagaraExt_VariableValue_DataInterface
    value {dataInterface=&self:NiagaraDataInterfaceSkeletalMesh_5}
   }
  }
  User.SkeletalMesh Wrist {
   type=NiagaraDataInterfaceSkeletalMesh
   defaultValue {
    struct=NiagaraExt_VariableValue_DataInterface
    value {dataInterface=&self:NiagaraDataInterfaceSkeletalMesh_0}
   }
  }
 ]
}
